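# Animation Blueprint for the BionicKitty backpack cosmetic, adjusting how the pack sits via its root bone.
# The AnimGraph feeds the input pose through two additive root-bone transforms with fixed offsets and a slight roll.
# On update, the EventGraph remaps the yaw of the owner mesh's spine_05-to-spine_01 delta from reference pose into BackpackRotate.
# The AnimGraph never reads that variable, and a debug print node sits unconnected.
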AnimBlueprint /BRCosmetics/Accessories/FORT_Backpacks/Backpack_F_MED_BionicKitty/Meshes/F_MED_BionicKitty_Pack_AnimBP
target_skeleton=&/F_MED_BionicKitty_Pack_Skeleton
graphs [
 AnimGraph {nodes [
   n1 = "Transform (Modify) Bone - Bone: root"(ComponentPose=Local To Component(LocalPose=Input Pose()), Translation="0,-0.5,0", Rotation="0,0,0") {
    rotationMode=BMM_Additive
    scaleMode=BMM_Additive
    translationMode=BMM_Additive
    alphaBoolBlend {blendOption=Linear}
    boneToModify {boneName=root}
    translation {y=-0.5}
   }
   n2 = "Transform (Modify) Bone - Bone: root"(ComponentPose=@n1, Translation="0,-0.163,7.6", Rotation="0.000,0.000,-5.000") {
    rotationMode=BMM_Additive
    scaleMode=BMM_Additive
    translationMode=BMM_Additive
    alphaBoolBlend {blendOption=Linear}
    boneToModify {boneName=root}
    rotation {roll=-5}
    translation {y=-0.163, z=7.6}
   }
   Output Pose(Result=Component To Local(ComponentPose=@n2)) {layerGroup=None}
  ]}
 EventGraph {nodes [
   Event BlueprintUpdateAnimation()
   n1 = BreakTransform(InTransform=GetDeltaTransformFromRefPose(self=GetOwnerSkeletalMeshComponent(), BoneName=spine_05, BaseName=spine_01))
   > Set BackpackRotate(BackpackRotate:double=MapRangeClamped(Value=BreakRotator(InRot=@n1.Rotation).Yaw, InRangeA=-25, InRangeB=-65)) -> (Output_Get:double)
   TryGetPawnOwner()
   PrintString!(execute, InString="To String (Float)"(InDouble=0), bPrintToScreen, bPrintToLog, TextColor, Duration, Key) -> (then)
   Reroute Node(InputPin:Rotator=@n1.Rotation) -> (OutputPin:Rotator)
  ]}
]
variables [BackpackRotate {type=number}]
